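# Sound cue for the above-platform ambience of the Skytube, playing an exterior border wind loop.
# A single looping wave player feeds the output at reduced volume, routed to an ambient outdoor-ducking sound class.
# Overridden attenuation uses a non-spatialized box volume with falloff, making it an area bed rather than a directional source.
# Occlusion tracing applies a smoothed low-pass filter, and non-focus volume is lowered.
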SoundCue /Figment_S09/Athena/Sounds/Ambient/Skytube/Skytube_Platform_Amb_Above_Cue
root=WavePlayer_0
nodes [{id=WavePlayer_0, properties {bLooping=true, soundWaveAssetPtr=/FNGameplayCues/Sounds/Fort_Ambient_Looping/Wind_Exterior/border_wind_05Loop}}]
properties {
 bOverrideAttenuation=true
 maxDistance=2000
 soundClassObject=/Game/Sounds/FortSoundClassesAndModes/Fort_Ambient_DuckOutdoors_Athena_PSM
 volumeMultiplier=0.6
 attenuationOverrides {
  attenuationShape=Box
  bEnableFocusInterpolation=true
  bEnableOcclusion=true
  bSpatialize=false
  falloffDistance=1000
  nonFocusVolumeAttenuation=0.4
  occlusionInterpolationTime=1
  occlusionLowPassFilterFrequency=2000
  occlusionTraceChannel=ECC_GameTraceChannel11
  stereoSpread=2000
  attenuationShapeExtents=<1000 1000 1000>
 }
}
