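# SoundCue for the looping ambient bed below a Skytube platform, in the Fort_Gadgets sound class.
# A Random node picks between two equally weighted, looping Grumpy platform wave players.
# The cue overrides attenuation with a long falloff, wide stereo spread and occlusion with low-pass filtering.
# Listener focus with interpolation lowers the volume when the sound is out of focus.
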SoundCue /Figment_S09/Athena/Sounds/Ambient/Skytube/Skytube_Platform_Amb_Below_Cue
root=Random_0
nodes [
 {id=Random_0, children [WavePlayer_0,WavePlayer_1], properties {weights [1,1]}}
 {id=WavePlayer_0, properties {bLooping=true, soundWaveAssetPtr=/Game/Athena/Sounds/Events/Dopey/Grumpy/Grumpy_Platform_Loop_01}}
 {id=WavePlayer_1, properties {bLooping=true, soundWaveAssetPtr=/Figment_S09/Athena/Sounds/Events/Dopey/Grumpy/Grumpy_Platform_Loop_02}}
]
properties {
 bOverrideAttenuation=true
 maxDistance=20000
 soundClassObject=&Fort_Gadgets
 volumeMultiplier=1
 attenuationOverrides {
  bEnableFocusInterpolation=true
  bEnableListenerFocus=true
  bEnableOcclusion=true
  falloffDistance=20000
  focusAttackInterpSpeed=0.5
  focusAzimuth=10
  focusReleaseInterpSpeed=2
  nonFocusAzimuth=30
  nonFocusVolumeAttenuation=0.25
  occlusionInterpolationTime=1
  occlusionLowPassFilterFrequency=2000
  occlusionTraceChannel=ECC_GameTraceChannel11
  reverbWetLevelMax=10
  reverbWetLevelMin=10
  stereoSpread=2000
  attenuationShapeExtents {x=0}
 }
}
